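; Carvable pumpkin quest actor (variant D) for the VoiceFork BR quests that plays a carving sound when a pawn interacts with it.
; The begin-interact event spawns the carve sound cue attached to the actor's Audio component, without checking the interaction type or pawn.
; BeginPlay, overlap, Tick and the construction script are empty, so this graph holds no quest or visual-change logic.
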(event EventBeginPlay)

(event Collision|EventActorBeginOverlap
  (params (OtherActor :type Actor%20Object%20Reference)))

(event EventTick
  (params (DeltaSeconds :type Float%20%28single-precision%29)))

(event Interaction|EventBlueprintOnBeginInteract
  (params (InteractingPawn :type Fort%20Pawn%20Object%20Reference) (InteractionBeingAttempted :type EInteractionBeingAttempted%20Enum :default "FirstInteraction") (InteractionType :type TInteractionType%20Enum :default "IT_NoInteraction"))
  ; Play squealching sound on interact
  (Audio|SpawnSoundAttached "/VoiceFork_BR_Quests/Items/QuestActors/Pumpkins/SFX/Quest_S37_VoiceFork_CarvablePumpkin_Carve_Cue.Quest_S37_VoiceFork_CarvablePumpkin_Carve_Cue" (Variables|Default|GetAudio)))

(fn ConstructionScript)
